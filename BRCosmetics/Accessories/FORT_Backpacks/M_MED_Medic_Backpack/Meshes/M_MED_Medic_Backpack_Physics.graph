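# Physics asset for the medic backpack cosmetic, giving collision bodies and joints to the pack and an articulated glove attached to it.
# The root body combines a box and a capsule, while the glove and its five two-segment fingers use one box per bone.
# Constraints chain glove base to root, glove mid to base and each finger to glove mid, with limited swing and very tight twist.
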PhysicsAsset /BRCosmetics/Accessories/FORT_Backpacks/M_MED_Medic_Backpack/Meshes/M_MED_Medic_Backpack_Physics
bodies {
 glove_base [{shapeType=Box, extentX=2.51, extentY=1.01, extentZ=7.01, center=<1.4107 -0.0108 -0.1159>, rotation=<rot 0 0 5.0001>}]
 glove_mid [{shapeType=Box, extentX=3.1228, extentY=0.8078, extentZ=6.3676, center=<1.487 -0.019 -0.2045>, rotation=<rot 0 0 5.0001>}]
 index_01 [{shapeType=Box, extentX=2, extentY=1.4888, extentZ=0.8886, center=<2.0053 0 0>}]
 index_02 [{shapeType=Box, extentX=2.7199, extentY=1.4888, extentZ=0.8886, center=<1.4665 0 0>}]
 mid_01 [{shapeType=Box, extentX=2.5, extentY=1.5, extentZ=0.6564, center=<2.1927 0 -0.0001>}]
 mid_02 [{shapeType=Box, extentX=3, extentY=1.5, extentZ=0.6564, center=<0.8477 0 0.0002>}]
 pinky_01 [{shapeType=Box, extentX=2, extentY=1, extentZ=0.7399, center=<1.7845 0.0663 0>, rotation=<rot 0 4.83 0>}]
 pinky_02 [{shapeType=Box, extentX=2, extentY=1, extentZ=0.7399, center=<1 0 0>, rotation=<rot 0 1.8524 0>}]
 ring_01 [{shapeType=Box, extentX=2, extentY=1.3838, extentZ=0.7928, center=<1.7196 0.1123 0>, rotation=<rot 0 3.7738 0>}]
 ring_02 [{shapeType=Box, extentX=2.629, extentY=1.3838, extentZ=0.7928, center=<1.3822 0.0183 0>, rotation=<rot 0 2.7425 0>}]
 root [
  {shapeType=Capsule, length=16.7971, radius=3.4881, center=<0.4068 -13.2394 -9.8836>, rotation=<rot -90 0 0>}
  {shapeType=Box, extentX=19.6592, extentY=15.1937, extentZ=31.9715, center=<-0.0534 -6.7374 -3.5439>, rotation=<rot 0 0 -2.4929>}
 ]
 thumb_01 [{shapeType=Box, extentX=2, extentY=1, extentZ=0.8265, center=<1 0 0>}]
 thumb_02 [{shapeType=Box, extentX=2, extentY=1, extentZ=0.8265, center=<1.3553 -0.1989 0>}]
}
constraints [
 {bone1Name=glove_base, bone2Name=root, swing1LimitDegrees=45, swing2LimitDegrees=45, twistLimitDegrees=1}
 {bone1Name=glove_mid, bone2Name=glove_base, swing1LimitDegrees=45, swing2LimitDegrees=45, twistLimitDegrees=1}
 {bone1Name=thumb_01, bone2Name=glove_mid, swing1LimitDegrees=20, swing2LimitDegrees=45, twistLimitDegrees=1}
 {bone1Name=thumb_02, bone2Name=thumb_01, swing1LimitDegrees=20, swing2LimitDegrees=45, twistLimitDegrees=1}
 {bone1Name=index_01, bone2Name=glove_mid, swing1LimitDegrees=20, swing2LimitDegrees=45, twistLimitDegrees=1}
 {bone1Name=index_02, bone2Name=index_01, swing1LimitDegrees=20, swing2LimitDegrees=45, twistLimitDegrees=1}
 {bone1Name=mid_01, bone2Name=glove_mid, swing1LimitDegrees=20, swing2LimitDegrees=45, twistLimitDegrees=1}
 {bone1Name=mid_02, bone2Name=mid_01, swing1LimitDegrees=20, swing2LimitDegrees=45, twistLimitDegrees=1}
 {bone1Name=pinky_01, bone2Name=glove_mid, swing1LimitDegrees=20, swing2LimitDegrees=45, twistLimitDegrees=1}
 {bone1Name=pinky_02, bone2Name=pinky_01, swing1LimitDegrees=20, swing2LimitDegrees=45, twistLimitDegrees=1}
 {bone1Name=ring_02, bone2Name=ring_01, swing1LimitDegrees=20, swing2LimitDegrees=45, twistLimitDegrees=1}
 {bone1Name=ring_01, bone2Name=glove_mid, swing1LimitDegrees=20, swing2LimitDegrees=45, twistLimitDegrees=1}
]
