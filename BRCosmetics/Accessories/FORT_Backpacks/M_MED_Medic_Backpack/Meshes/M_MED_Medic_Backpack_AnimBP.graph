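# Animation Blueprint for the Medic backpack cosmetic, adding physics-driven secondary motion to its dangling parts.
# From the reference pose, a RigidBody node simulates the backpack's physics asset under overridden world gravity.
# Three chained Anim Dynamics nodes then swing the two stethoscope bones and the zipper within angular limits, with heavy damping and reduced component-motion influence on the stethoscope.
# The event graph holds no logic.
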AnimBlueprint /BRCosmetics/Accessories/FORT_Backpacks/M_MED_Medic_Backpack/Meshes/M_MED_Medic_Backpack_AnimBP
target_skeleton=&/M_MED_Medic_Backpack_Skeleton
graphs [
 AnimGraph {nodes [
   n1 = RigidBody(ComponentPose=Local To Component(LocalPose=Local Space Ref Pose())) {
    bOverrideWorldGravity=true
    overridePhysicsAsset=&/M_MED_Medic_Backpack_Physics
    alphaBoolBlend {blendOption=Linear}
    componentLinearAccScale=<1 1 1>
    componentLinearVelScale=<1 1 1>
    overrideWorldGravity {z=-500}
   }
   n2 = Anim Dynamics - Bone: stethoscope_01(ComponentPose=@n1, Alpha=1, ExternalForce="0,0,0") {
    angularDampingOverride=0.999
    bGravityOverrideInSimSpace=true
    bOverrideAngularDamping=true
    alphaBoolBlend {blendOption=Linear}
    boundBone {boneName=stethoscope_01}
    componentLinearAccScale=<0.2 0.2 0.2>
    componentLinearVelScale=<0.2 0.2 0.2>
    physicsBodyDefinitions [{
      sphereCollisionRadius=0
      boundBone {boneName=stethoscope_01}
      boxExtents=<2 10 10>
      constraintSetup {angularLimitsMax=<0 45 20>, angularLimitsMin=<0 0 -20>, angularTarget=<0 0 0>}
      localJointOffset=<2 0 0>
     }]
   }
   n3 = Anim Dynamics - Bone: stethoscope_02(ComponentPose=@n2, Alpha=1, ExternalForce="0,0,0") {
    angularDampingOverride=0.999
    bGravityOverrideInSimSpace=true
    bOverrideAngularDamping=true
    alphaBoolBlend {blendOption=Linear}
    boundBone {boneName=stethoscope_02}
    componentLinearAccScale=<0.2 0.2 0.2>
    componentLinearVelScale=<0.2 0.2 0.2>
    physicsBodyDefinitions [{
      sphereCollisionRadius=0
      boundBone {boneName=stethoscope_02}
      boxExtents=<2 10 10>
      constraintSetup {angularLimitsMax=<0 45 20>, angularLimitsMin=<0 0 -20>, angularTarget=<0 0 0>}
      localJointOffset=<2 0 0>
     }]
   }
   n4 = Anim Dynamics - Bone: zipper(ComponentPose=@n3, Alpha=1, ExternalForce="0,0,0") {
    angularDampingOverride=0.98
    bGravityOverrideInSimSpace=true
    bOverrideAngularDamping=true
    alphaBoolBlend {blendOption=Linear}
    boundBone {boneName=zipper}
    componentLinearAccScale=<1 1 1>
    componentLinearVelScale=<1 1 1>
    physicsBodyDefinitions [{
      sphereCollisionRadius=0
      boundBone {boneName=zipper}
      boxExtents=<5 4 5>
      constraintSetup {angularLimitsMax=<0 40 120>, angularLimitsMin=<0 -40 0>, angularTarget=<0 0 0>}
      localJointOffset=<1.5 0 0>
     }]
   }
   Output Pose(Result=Component To Local(ComponentPose=@n4)) {layerGroup=None}
  ]}
 EventGraph {nodes [Event BlueprintUpdateAnimation()]}
]
variables []
